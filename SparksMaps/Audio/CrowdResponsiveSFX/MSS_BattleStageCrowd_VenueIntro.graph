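# One-shot stereo MetaSound source that plays an arena crowd bed, named for the battle stage venue intro.
# A wave player starts the festival crowd loop asset on play, triggering an ADSR envelope attack and a delay timer.
# The timer later triggers the release, so the crowd fades in, holds for a fixed time, then fades out.
# The envelope scales both channels, and its completion signals that the source has finished.
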MetaSoundSource /SparksMaps/Audio/CrowdResponsiveSFX/MSS_BattleStageCrowd_VenueIntro
inputs [UE.Source.OnPlay {type=Trigger}]
interfaces [UE.OutputFormat.Stereo:1.0, UE.Source.OneShot:1.0, UE.Source:1.1]
outputs [
 "UE.OutputFormat.Stereo.Audio:0" {type=Audio}
 "UE.OutputFormat.Stereo.Audio:1" {type=Audio}
 UE.Source.OneShot.OnFinished {type=Trigger}
]
pages [Default {
  input_defaults {UE.Source.OnPlay=null}
  nodes [
   WavePlayer_0 = UE.Wave Player:Stereo(Play=^UE.Source.OnPlay) {also {defaults {Loop="<MetaSoundBuilderResult.SUCCEEDED: 0>", Wave Asset=/SparksMaps/Audio/Wavs/Festival_CrowdArenaGreat_Loop}, version [1,1]}}
   TriggerDelay_0 = UE.Trigger Delay(In=@WavePlayer_0.On Play) {also {defaults {Delay Time=27}}}
   ADSREnvelopeADSREnvelope_0 = ADSR Envelope.ADSR Envelope:Float(Trigger Release=@TriggerDelay_0.Out, Trigger Attack=@WavePlayer_0.On Play) {also {defaults {Attack Time=2, Release Time=5, Sustain Level=1}}}
   Multiply_0 = UE.Multiply:Audio by Float(AdditionalOperands=@ADSREnvelopeADSREnvelope_0.Out Envelope, PrimaryOperand=@WavePlayer_0.Out Right)
   Multiply_1 = UE.Multiply:Audio by Float(AdditionalOperands=@ADSREnvelopeADSREnvelope_0.Out Envelope, PrimaryOperand=@WavePlayer_0.Out Left)
  ]
  links [
   {dst="@UE.Source.OneShot.OnFinished", src=ADSREnvelopeADSREnvelope_0.On Done}
   {dst="@UE.OutputFormat.Stereo.Audio:1", src=Multiply_0.Out}
   {dst="@UE.OutputFormat.Stereo.Audio:0", src=Multiply_1.Out}
  ]
 }]
properties {
 assetImportData=&self:AssetImportData
 numChannels=2
 outputFormat=Stereo
 sampleRate=48000
 volume=0.3
}
