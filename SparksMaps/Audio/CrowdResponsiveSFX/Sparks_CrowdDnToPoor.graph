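# One-shot stereo MetaSound for responsive crowd SFX in Sparks maps, named for the crowd's drop to a poor state.
# On play, two random array picks each feed a stereo wave player, and a mixer layers them with one layer's gain raised.
# The boosted layer alone signals OnFinished, so its length sets the sound's lifetime.
# It uses large-crowd attenuation and a concurrency override allowing one instance at a time.
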MetaSoundSource /SparksMaps/Audio/CrowdResponsiveSFX/Sparks_CrowdDnToPoor
inputs [UE.Source.OnPlay {type=Trigger}]
interfaces [UE.OutputFormat.Stereo:1.0, UE.Source.OneShot:1.0, UE.Source:1.1]
outputs [
 "UE.OutputFormat.Stereo.Audio:0" {type=Audio}
 "UE.OutputFormat.Stereo.Audio:1" {type=Audio}
 UE.Source.OneShot.OnFinished {type=Trigger}
]
pages [Default {
  input_defaults {UE.Source.OnPlay=null}
  nodes [
   ArrayRandomGet_0 = Array.Random Get:WaveAsset:Array(Next=^UE.Source.OnPlay, In Array=VariableAccessor.WaveAsset:Array(Variable=InitVariable.WaveAsset:Array().Variable).Value)
   WavePlayer_0 = UE.Wave Player:Stereo(Play=@ArrayRandomGet_0.On Next, Wave Asset=@ArrayRandomGet_0.Value) {also {version [1,1]}}
   ArrayRandomGet_1 = Array.Random Get:WaveAsset:Array(Next=^UE.Source.OnPlay, In Array=VariableAccessor.WaveAsset:Array(Variable=InitVariable.WaveAsset:Array().Variable).Value)
   WavePlayer_1 = UE.Wave Player:Stereo(Play=@ArrayRandomGet_1.On Next, Wave Asset=@ArrayRandomGet_1.Value) {also {version [1,1]}}
   AudioMixerAudioMixerStereo2_0 = "AudioMixer.Audio Mixer (Stereo, 2)"(In 1 L=@WavePlayer_0.Out Left, In 1 R=@WavePlayer_0.Out Right, In 0 L=@WavePlayer_1.Out Left, In 0 R=@WavePlayer_1.Out Right) {also {defaults {Gain 1=2}}}
  ]
  links [
   {dst="@UE.OutputFormat.Stereo.Audio:0", src=AudioMixerAudioMixerStereo2_0.Out L}
   {dst="@UE.OutputFormat.Stereo.Audio:1", src=AudioMixerAudioMixerStereo2_0.Out R}
   {dst="@UE.Source.OneShot.OnFinished", src=WavePlayer_0.On Finished}
  ]
 }]
properties {
 assetImportData=&self:AssetImportData
 attenuationSettings=/SparksMaps/Audio/FM_Attenu_Crowd_Lg
 bOverrideConcurrency=true
 maxDistance=10000
 numChannels=2
 outputFormat=Stereo
 sampleRate=48000
 soundClassObject=&Fort_SFX_InGame
 volume=0.5
 concurrencyOverrides {maxCount=1}
}
